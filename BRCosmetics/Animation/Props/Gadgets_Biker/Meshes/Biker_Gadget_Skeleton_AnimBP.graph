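# Animation Blueprint for the Biker gadget prop skeleton, which shifts the front tire bone to line up with a kick drum mesh.
# The pose blends InMotion and default slots on bIsMoving2D, then adds a Y offset to JNT_Front_Tire_offset.
# For one second after initialization, updates find emote mesh 3004 or 3003 and derive that offset from drum socket distance.
# A front tire rotation chain exists but is unconnected to the output.
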AnimBlueprint /BRCosmetics/Animation/Props/Gadgets_Biker/Meshes/Biker_Gadget_Skeleton_AnimBP
target_skeleton=&/Biker_Gadget_Skeleton
graphs [
 AnimGraph {nodes [
   Component To Local(ComponentPose="Transform (Modify) Bone - Bone: JNT_Front_Tire"!(
     ComponentPose=Local To Component()
     Alpha=1
     Translation="0,0,0"
     Rotation_Roll:float=0
     Rotation_Pitch:float=0
     Rotation_Yaw:float=Get Wheel Rotation() -> (Wheel Rotation:double)
     Scale="1,1,1"
    ) -> (Pose) {alphaBoolBlend {blendOption=Linear}, boneToModify {boneName=JNT_Front_Tire}})
   n1 = Local To Component(LocalPose=Blend Poses by bool(BlendPose_0=Slot 'InMotion'() {slotName=InMotion}, BlendPose_1=Slot 'DefaultSlot'(), bActiveValue=Get bIsMoving2D()) {blendType=Linear})
   n2 = MakeVector(Y=Get FrontWheelTranslationOffset() -> (FrontWheelTranslationOffset:double))
   Output Pose(Result=Component To Local(ComponentPose="Transform (Modify) Bone - Bone: JNT_Front_Tire_offset"!(ComponentPose=@n1, Alpha=1, Translation=@n2, Rotation_Roll:float=0, Rotation_Pitch:float=0, Rotation_Yaw:float=0, Scale="1,1,1") -> (Pose) {translationMode=BMM_Additive, alphaBoolBlend {blendOption=Linear}, boneToModify {boneName=JNT_Front_Tire_offset}}))
  ]}
 EventGraph {nodes [
   Event BlueprintUpdateAnimation()
   > Branch(Condition=Get WheelSet() -> (WheelSet:bool))
   > n1 = Is Valid(InputObject=Get Drum() -> (Drum:Mesh Component Object Reference))
   n2 = Get EmoteMeshCompDatas!(self:Fort Pawn Object Reference=Get As Player Pawn Athena() -> (As Player Pawn Athena:Player Pawn Athena Object Reference)) -> (EmoteMeshCompDatas:Map of Integers to Emote Mesh Comp Data Structures)
   n3 = Find!(TargetMap:Map of Integers to Emote Mesh Comp Data Structures=@n2, Key:"Integer (by ref)"=3004) -> (
    Value_MeshComp:Mesh Component Object Reference
    Value_PropExclusionTag:Gameplay Tag Structure
    Value_bTickEvenWhenNotRendered:bool=~
    ReturnValue
   )
   n4 = Branch(execute=@n1.Is Not Valid, Condition=@n3.ReturnValue)
   > Set Drum(Drum:Mesh Component Object Reference=@n3) -> (Output_Get:Mesh Component Object Reference)
   n5 = Find(TargetMap=@n2, Key=3003) -> (Value_bTickEvenWhenNotRendered=~)
   Branch(execute=@n4.else, Condition=@n5.ReturnValue)
   > Set Drum(Drum=@n5)
   n6 = Get Drum()
   n7 = GetSocketLocation!(self:Scene Component Object Reference=@n6, InSocketName:Name=kickdrum_head) -> (ReturnValue:Vector)
   n8 = Get World Rotation!(self:Scene Component Object Reference=@n6) -> (ReturnValue:Rotator) {is=title K2Node_CallFunction}
   n9 = vector + vector(
    A=GetSocketLocation(self=@n6, InSocketName=FX_KickDrum_Out)
    B=vector * vector(
     A=Get World Scale!(self:Scene Component Object Reference=@n6) -> (ReturnValue:Vector) {is=title K2Node_CallFunction}
     B:Vector=Rotate Vector!(A:Vector="0,0,5.6283", B:Rotator=@n8) -> (ReturnValue:Vector)
    ) -> (ReturnValue=~)
   )
   Set FrontWheelTranslationOffset(
    execute=@n1.Is Valid
    FrontWheelTranslationOffset:double=float - float(
     A="Distance (Vector)"!(V1:Vector=@n9, V2:Vector=@n7) -> (ReturnValue:double)
     B=Get FrontWheelTranslationOffsetModifier() -> (FrontWheelTranslationOffsetModifier:double)
    )
   ) -> (Output_Get:double)
   Event BlueprintInitializeAnimation()
   > n10 = Cast To PlayerPawn_Athena(Object=GetOwningActor())
   > Set As Player Pawn Athena(As Player Pawn Athena:Player Pawn Athena Object Reference=@n10) -> (Output_Get:Player Pawn Athena Object Reference)
   > Set WheelSet(WheelSet:bool=true) -> (Output_Get:bool)
   > Delay(Duration=1)
   > Set WheelSet(WheelSet=false)
  ]}
]
variables [
 Actor Rotation {type=number}
 As Player Pawn Athena {type=PlayerPawn_Athena_C}
 Drum {type=MeshComponent, default=null}
 FrontWheelTranslationOffset {type=number}
 FrontWheelTranslationOffsetModifier {type=number, default=6}
 OwningActor {type=Actor, default=null}
 Wheel Rotation {type=number}
 WheelSet {type=boolean}
]
